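# Animation Blueprint for the HomeRange backpack that adds physics-driven secondary motion to its seven dyn_bullet bones.
# The cached input pose feeds two chains of per-bone Anim Dynamics nodes: a standard one and a wind-enabled skydiving one with its own limits and tuning.
# A bool blend on bIsSkyDiving picks the chain.
# Each update sets the shared gravity override for grounded, freefall or open-parachute states, weakest in freefall.
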AnimBlueprint /BRCosmetics/Accessories/FORT_Backpacks/Backpack_F_MED_HomeRange/Meshes/F_MED_HomeRange_Pack_AnimBP
target_skeleton=&/F_MED_HomeRange_Pack_Skeleton
graphs [
 AnimGraph {nodes [
   n1 = Get Gravity Override()
   n2 = Anim Dynamics - Bone: dyn_bullet_07!(ComponentPose=Local To Component(LocalPose=Use cached pose 'InputPose'()), Alpha=1, GravityOverride:Vector=@n1, ExternalForce="0,0,0") -> (Pose) {
    angularDampingOverride=0.98
    angularSpringConstant=2550
    bOverrideAngularDamping=true
    bUseGravityOverride=true
    gravityScale=0.35
    alphaBoolBlend {blendOption=Linear}
    boundBone {boneName=dyn_bullet_07}
    componentAppliedLinearAccClamp=<1000 1000 1000>
    componentLinearAccScale=<0.5 0.5 0.5>
    componentLinearVelScale=<0.5 0.5 0.5>
    gravityOverride {z=-980}
    physicsBodyDefinitions [{
      boundBone {boneName=dyn_bullet_07}
      boxExtents=<11 11 14>
      constraintSetup {angularLimitsMax=<20 80 45>, angularLimitsMin=<-20 -40 -35>, angularTarget=<0 0 0>}
      localJointOffset=<1 0 0>
     }]
   }
   n3 = Anim Dynamics - Bone: dyn_bullet_06!(ComponentPose=@n2, Alpha=1, GravityOverride:Vector=@n1, ExternalForce="0,0,0") -> (Pose) {
    angularDampingOverride=0.98
    angularSpringConstant=2550
    bOverrideAngularDamping=true
    bUseGravityOverride=true
    gravityScale=0.35
    alphaBoolBlend {blendOption=Linear}
    boundBone {boneName=dyn_bullet_06}
    componentAppliedLinearAccClamp=<1000 1000 1000>
    componentLinearAccScale=<0.5 0.5 0.5>
    componentLinearVelScale=<0.5 0.5 0.5>
    gravityOverride {z=-980}
    physicsBodyDefinitions [{
      boundBone {boneName=dyn_bullet_06}
      boxExtents=<9 9 14>
      constraintSetup {angularLimitsMax=<20 75 45>, angularLimitsMin=<-20 -40 -35>, angularTarget=<0 0 0>}
      localJointOffset=<1 0 0>
     }]
   }
   n4 = Anim Dynamics - Bone: dyn_bullet_05!(ComponentPose=@n3, Alpha=1, GravityOverride:Vector=@n1, ExternalForce="0,0,0") -> (Pose) {
    angularDampingOverride=0.98
    angularSpringConstant=2550
    bOverrideAngularDamping=true
    bUseGravityOverride=true
    gravityScale=0.35
    alphaBoolBlend {blendOption=Linear}
    boundBone {boneName=dyn_bullet_05}
    componentAppliedLinearAccClamp=<1000 1000 1000>
    componentLinearAccScale=<0.5 0.5 0.5>
    componentLinearVelScale=<0.5 0.5 0.5>
    gravityOverride {z=-980}
    physicsBodyDefinitions [{
      boundBone {boneName=dyn_bullet_05}
      boxExtents=<10 10 14>
      constraintSetup {angularLimitsMax=<20 60 45>, angularLimitsMin=<-20 -22 -35>, angularTarget=<0 0 0>}
      localJointOffset=<1 0 0>
     }]
   }
   n5 = Anim Dynamics - Bone: dyn_bullet_04!(ComponentPose=@n4, Alpha=1, GravityOverride:Vector=@n1, ExternalForce="0,0,0") -> (Pose) {
    angularDampingOverride=0.98
    angularSpringConstant=2550
    bOverrideAngularDamping=true
    bUseGravityOverride=true
    gravityScale=0.35
    alphaBoolBlend {blendOption=Linear}
    boundBone {boneName=dyn_bullet_04}
    componentAppliedLinearAccClamp=<1000 1000 1000>
    componentLinearAccScale=<0.5 0.5 0.5>
    componentLinearVelScale=<0.5 0.5 0.5>
    gravityOverride {z=-980}
    physicsBodyDefinitions [{
      boundBone {boneName=dyn_bullet_04}
      boxExtents=<9 9 14>
      constraintSetup {angularLimitsMax=<-20 30 45>, angularLimitsMin=<20 -65 -35>, angularTarget=<0 0 0>}
      localJointOffset=<1 0 0>
     }]
   }
   n6 = Anim Dynamics - Bone: dyn_bullet_03!(ComponentPose=@n5, Alpha=1, GravityOverride:Vector=@n1, ExternalForce="0,0,0") -> (Pose) {
    angularDampingOverride=0.98
    angularSpringConstant=2550
    bOverrideAngularDamping=true
    bUseGravityOverride=true
    gravityScale=0.35
    alphaBoolBlend {blendOption=Linear}
    boundBone {boneName=dyn_bullet_03}
    componentAppliedLinearAccClamp=<1000 1000 1000>
    componentLinearAccScale=<0.5 0.5 0.5>
    componentLinearVelScale=<0.5 0.5 0.5>
    gravityOverride {z=-980}
    physicsBodyDefinitions [{
      boundBone {boneName=dyn_bullet_03}
      boxExtents=<11 11 14>
      constraintSetup {angularLimitsMax=<20 40 45>, angularLimitsMin=<-20 -75 -35>, angularTarget=<0 0 0>}
      localJointOffset=<1 0 0>
     }]
   }
   n7 = Anim Dynamics - Bone: dyn_bullet_02!(ComponentPose=@n6, Alpha=1, GravityOverride:Vector=@n1, ExternalForce="0,0,0") -> (Pose) {
    angularDampingOverride=0.98
    angularSpringConstant=2550
    bOverrideAngularDamping=true
    bUseGravityOverride=true
    gravityScale=0.35
    alphaBoolBlend {blendOption=Linear}
    boundBone {boneName=dyn_bullet_02}
    componentAppliedLinearAccClamp=<1000 1000 1000>
    componentLinearAccScale=<0.5 0.5 0.5>
    componentLinearVelScale=<0.5 0.5 0.5>
    gravityOverride {z=-980}
    physicsBodyDefinitions [{
      boundBone {boneName=dyn_bullet_02}
      boxExtents=<10 10 14>
      constraintSetup {angularLimitsMax=<5 5 5>, angularLimitsMin=<-5 -5 -5>, angularTarget=<0 0 0>}
      localJointOffset=<1 0 0>
     }]
   }
   n8 = Anim Dynamics - Bone: dyn_bullet_01!(ComponentPose=@n7, Alpha=1, GravityOverride:Vector=@n1, ExternalForce="0,0,0") -> (Pose) {
    angularDampingOverride=0.98
    angularSpringConstant=2550
    bOverrideAngularDamping=true
    bUseGravityOverride=true
    gravityScale=0.35
    alphaBoolBlend {blendOption=Linear}
    boundBone {boneName=dyn_bullet_01}
    componentAppliedLinearAccClamp=<1000 1000 1000>
    componentLinearAccScale=<0.5 0.5 0.5>
    componentLinearVelScale=<0.5 0.5 0.5>
    gravityOverride {z=-980}
    physicsBodyDefinitions [{
      boundBone {boneName=dyn_bullet_01}
      boxExtents=<10 10 14>
      constraintSetup {angularLimitsMax=<15 35 40>, angularLimitsMin=<-15 -75 -30>, angularTarget=<0 0 0>}
      localJointOffset=<1 0 0>
     }]
   }
   n9 = Get Gravity Override()
   n10 = Anim Dynamics - Bone: dyn_bullet_07(ComponentPose=Local To Component(LocalPose=Use cached pose 'InputPose'()), Alpha=1, GravityOverride=@n9, ExternalForce="0,0,0") {
    angularBiasOverride=0.999
    angularDampingOverride=0.999
    angularSpringConstant=2550
    bEnableWind=true
    bOverrideAngularBias=true
    bOverrideAngularDamping=true
    bUseGravityOverride=true
    gravityScale=0.35
    windScale=100
    alphaBoolBlend {blendOption=Linear}
    boundBone {boneName=dyn_bullet_07}
    componentAppliedLinearAccClamp=<2000 2000 2000>
    componentLinearAccScale=<0.4 0.4 0.4>
    componentLinearVelScale=<0.4 0.4 0.4>
    gravityOverride {z=-980}
    physicsBodyDefinitions [{
      boundBone {boneName=dyn_bullet_07}
      boxExtents=<13 13 13>
      constraintSetup {angularLimitsMax=<20 21 40>, angularLimitsMin=<-20 -21 -80>, angularTarget=<0 0 0>}
      localJointOffset=<2 0 0>
     }]
    simSpaceSettings {externalAngularVelocity {y=2}}
   }
   n11 = Anim Dynamics - Bone: dyn_bullet_06(ComponentPose=@n10, Alpha=1, GravityOverride=@n9, ExternalForce="0,0,0") {
    angularBiasOverride=0.999
    angularDampingOverride=0.999
    angularSpringConstant=2550
    bEnableWind=true
    bOverrideAngularBias=true
    bOverrideAngularDamping=true
    bUseGravityOverride=true
    gravityScale=0.35
    windScale=100
    alphaBoolBlend {blendOption=Linear}
    boundBone {boneName=dyn_bullet_06}
    componentAppliedLinearAccClamp=<2000 2000 2000>
    componentLinearAccScale=<0.4 0.4 0.4>
    componentLinearVelScale=<0.4 0.4 0.4>
    gravityOverride {z=-980}
    physicsBodyDefinitions [{
      boundBone {boneName=dyn_bullet_06}
      boxExtents=<8 8 8>
      constraintSetup {angularLimitsMax=<20 19 40>, angularLimitsMin=<-20 -19 -75>, angularTarget=<0 0 0>}
      localJointOffset=<2 0 0>
     }]
    simSpaceSettings {externalAngularVelocity {y=2}}
   }
   n12 = Anim Dynamics - Bone: dyn_bullet_05(ComponentPose=@n11, Alpha=1, GravityOverride=@n9, ExternalForce="0,0,0") {
    angularBiasOverride=0.999
    angularDampingOverride=0.999
    angularSpringConstant=2550
    bEnableWind=true
    bOverrideAngularBias=true
    bOverrideAngularDamping=true
    bUseGravityOverride=true
    gravityScale=0.35
    windScale=100
    alphaBoolBlend {blendOption=Linear}
    boundBone {boneName=dyn_bullet_05}
    componentAppliedLinearAccClamp=<2000 2000 2000>
    componentLinearAccScale=<0.4 0.4 0.4>
    componentLinearVelScale=<0.4 0.4 0.4>
    gravityOverride {z=-980}
    physicsBodyDefinitions [{
      boundBone {boneName=dyn_bullet_05}
      boxExtents=<10 10 10>
      constraintSetup {angularLimitsMax=<20 20 45>, angularLimitsMin=<-20 -20 -80>, angularTarget=<0 0 0>}
      localJointOffset=<2 0 0>
     }]
    simSpaceSettings {externalAngularVelocity {y=2}}
   }
   n13 = Anim Dynamics - Bone: dyn_bullet_04(ComponentPose=@n12, Alpha=1, GravityOverride=@n9, ExternalForce="0,0,0") {
    angularBiasOverride=0.999
    angularDampingOverride=0.999
    angularSpringConstant=2550
    bEnableWind=true
    bOverrideAngularBias=true
    bOverrideAngularDamping=true
    bUseGravityOverride=true
    gravityScale=0.35
    windScale=100
    alphaBoolBlend {blendOption=Linear}
    boundBone {boneName=dyn_bullet_04}
    componentAppliedLinearAccClamp=<2000 2000 2000>
    componentLinearAccScale=<0.4 0.4 0.4>
    componentLinearVelScale=<0.4 0.4 0.4>
    gravityOverride {z=-980}
    physicsBodyDefinitions [{
      boundBone {boneName=dyn_bullet_04}
      boxExtents=<6 6 6>
      constraintSetup {angularLimitsMax=<-20 20 45>, angularLimitsMin=<20 -20 -75>, angularTarget=<0 0 0>}
      localJointOffset=<2 0 0>
     }]
    simSpaceSettings {externalAngularVelocity {y=2}}
   }
   n14 = Anim Dynamics - Bone: dyn_bullet_03(ComponentPose=@n13, Alpha=1, GravityOverride=@n9, ExternalForce="0,0,0") {
    angularBiasOverride=0.999
    angularDampingOverride=0.999
    angularSpringConstant=2550
    bEnableWind=true
    bOverrideAngularBias=true
    bOverrideAngularDamping=true
    bUseGravityOverride=true
    gravityScale=0.35
    windScale=100
    alphaBoolBlend {blendOption=Linear}
    boundBone {boneName=dyn_bullet_03}
    componentAppliedLinearAccClamp=<2000 2000 2000>
    componentLinearAccScale=<0.4 0.4 0.4>
    componentLinearVelScale=<0.4 0.4 0.4>
    gravityOverride {z=-980}
    physicsBodyDefinitions [{
      boundBone {boneName=dyn_bullet_03}
      boxExtents=<11 11 11>
      constraintSetup {angularLimitsMax=<20 19 45>, angularLimitsMin=<-20 -19 -70>, angularTarget=<0 0 0>}
      localJointOffset=<2 0 0>
     }]
    simSpaceSettings {externalAngularVelocity {y=2}}
   }
   n15 = Anim Dynamics - Bone: dyn_bullet_02(ComponentPose=@n14, Alpha=1, GravityOverride=@n9, ExternalForce="0,0,0") {
    angularBiasOverride=0.999
    angularDampingOverride=0.999
    angularSpringConstant=2550
    bEnableWind=true
    bOverrideAngularBias=true
    bOverrideAngularDamping=true
    bUseGravityOverride=true
    gravityScale=0.35
    windScale=100
    alphaBoolBlend {blendOption=Linear}
    boundBone {boneName=dyn_bullet_02}
    componentAppliedLinearAccClamp=<2000 2000 2000>
    componentLinearAccScale=<0.4 0.4 0.4>
    componentLinearVelScale=<0.4 0.4 0.4>
    gravityOverride {z=-980}
    physicsBodyDefinitions [{
      boundBone {boneName=dyn_bullet_02}
      boxExtents=<6 6 6>
      constraintSetup {angularLimitsMax=<5 5 5>, angularLimitsMin=<-5 -5 -30>, angularTarget=<0 0 0>}
      localJointOffset=<2 0 0>
     }]
    simSpaceSettings {externalAngularVelocity {y=2}}
   }
   n16 = Anim Dynamics - Bone: dyn_bullet_01(ComponentPose=@n15, Alpha=1, GravityOverride=@n9, ExternalForce="0,0,0") {
    angularBiasOverride=0.999
    angularDampingOverride=0.999
    angularSpringConstant=2550
    bEnableWind=true
    bOverrideAngularBias=true
    bOverrideAngularDamping=true
    bUseGravityOverride=true
    gravityScale=0.35
    windScale=100
    alphaBoolBlend {blendOption=Linear}
    boundBone {boneName=dyn_bullet_01}
    componentAppliedLinearAccClamp=<2000 2000 2000>
    componentLinearAccScale=<0.4 0.4 0.4>
    componentLinearVelScale=<0.4 0.4 0.4>
    gravityOverride {z=-980}
    physicsBodyDefinitions [{
      boundBone {boneName=dyn_bullet_01}
      boxExtents=<8 8 8>
      constraintSetup {angularLimitsMax=<15 17 40>, angularLimitsMin=<-15 -17 -50>, angularTarget=<0 0 0>}
      localJointOffset=<2 0 0>
     }]
    simSpaceSettings {externalAngularVelocity {y=2}}
   }
   Output Pose(Result=Two Way Blend!(A=Component To Local(ComponentPose=@n8), B=Component To Local(ComponentPose=@n16), bAlphaBoolEnabled:bool=Get bIsSkyDiving()) -> (Pose) {blendNode {alphaInputType=Bool, alphaBoolBlend {blendOption=Linear}}}) {layerGroup=None}
   Save cached pose 'InputPose'(Pose=Input Pose()) {cacheName=InputPose}
  ]}
 EventGraph {nodes [
   Event BlueprintUpdateAnimation()
   > n1 = Branch(Condition=Get bIsSkyDiving())
   > n2 = Branch(Condition=NOT Boolean(A=Get bIsParachuteOpened()))
   > Set Gravity Override(Gravity Override="0,0,-500")
   Set Gravity Override(execute=@n2.else, Gravity Override="0,0,-800")
   Set Gravity Override(execute=@n1.else, Gravity Override="0,0,-980")
  ]}
]
variables [Gravity Override {type=Vector, default=<0 0 -100>}]
